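# Post-process animation Blueprint for the RobustTorn large male head skeleton, layering a Control Rig and lip corrections over the input pose.
# The event graph turns movement, emote, skydive, swim and front-end flags into an Idle FE value, passed to the rig only as active during front-end idle.
# It also multiplies remapped jaw-open and frown curves into per-side alphas, which additively offset the outer lower lip bones in bone space.
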AnimBlueprint /BRCosmetics/Characters/Player/Male/Large/Heads/M_LRG_RobustTorn_Head/Meshes/Fortnite_Base_Head_RobustTorn_Skeleton_AnimBlueprint_Post
target_skeleton=&/Fortnite_Base_Head_RobustTorn_Skeleton
graphs [
 AnimGraph {nodes [
   n1 = Control Rig!(Source=Input Pose(), Alpha, Idle FE:double=Get Idle FE() -> (Idle FE:double)) -> (Pose) {controlRigAssetReference {blueprintRigClass=&/M_LRG_RobustTorn_Head_CtrlRig.M_LRG_RobustTorn_Head_CtrlRig_C}}
   n2 = "Transform (Modify) Bone - Bone: L_lip_lower_outer"(
    ComponentPose=Local To Component(LocalPose=@n1)
    Alpha=Get L_Frown() -> (L_Frown:double)
    Translation="0,0,-5"
    Rotation="0,0,0"
    Scale="1,1,1"
   ) {
    rotationMode=BMM_Additive
    rotationSpace=BCS_BoneSpace
    translationMode=BMM_Additive
    translationSpace=BCS_BoneSpace
    alphaScaleBiasClamp {bInterpResult=true}
    boneToModify {boneName=L_lip_lower_outer}
    translation {z=-5}
   }
   n3 = "Transform (Modify) Bone - Bone: R_lip_lower_outer"(ComponentPose=@n2, Alpha=Get R_Frown() -> (R_Frown:double), Translation="0,0,-5", Rotation="0,0,0", Scale="1,1,1") {
    rotationMode=BMM_Additive
    rotationSpace=BCS_BoneSpace
    translationMode=BMM_Additive
    translationSpace=BCS_BoneSpace
    alphaScaleBiasClamp {bInterpResult=true}
    boneToModify {boneName=R_lip_lower_outer}
    translation {z=-5}
   }
   Output Pose(Result=Component To Local(ComponentPose=@n3))
  ]}
 EventGraph {nodes [
   Event BlueprintUpdateAnimation()
   n1 = NOT Boolean(A=Get bIsSurfaceSwimming())
   n2 = NOT Boolean(A=Get bIsPlayingEmote())
   n3 = NOT Boolean(A=Get bIsMoving2D())
   n4 = NOT Boolean(A=Get bIsSkyDiving())
   n5 = Get bIsInFrontEnd()
   > Set Idle FE(Idle FE:double="To Float (Boolean)"(InBool=AND Boolean!(A=@n3, B=@n2, C:bool=@n4, D:bool=@n1, E:bool=@n5) -> (ReturnValue))) -> (Output_Get:double)
   n6 = MapRangeClamped(Value=GetCurveValue(CurveName=jaw_open_pose), InRangeA=0, InRangeB=0.7)
   > Set L_Frown(L_Frown:double=float * float(A=@n6, B=GetCurveValue(CurveName=L_frown_pose)) {is=title K2Node_PromotableOperator}) -> (Output_Get:double)
   > Set R_Frown(R_Frown:double=float * float(A=@n6, B=GetCurveValue(CurveName=R_frown_pose))) -> (Output_Get:double)
   TryGetPawnOwner()
  ]}
]
variables [
 Idle FE {type=number}
 JawOpen {type=number}
 L_Frown {type=number}
 R_Frown {type=number}
]
